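; Boost cosmetic actor for the Boba vehicle boost, showing a Niagara effect while the vehicle boosts.
; Boost start and finish events switch the Niagara component active and inactive without resetting it.
; When the cosmetic color is ready, the shared vehicle cosmetics function library applies it to the Niagara component through the "PaintedColor" name, along with the additional parameter info.
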(event Vehicle|Cosmetics|EventOnBoostStarted
  (Components|Activation|SetActive (Variables|Default|GetNiagara) true false))

(event Vehicle|Cosmetics|EventOnBoostFinished
  (Components|Activation|SetActive (Variables|Default|GetNiagara) false false))

(event Vehicle|Cosmetics|EventOnColorReady
  (params (Color :type Name :default "None") (AdditionalInfo :type Vehicle%20Cosmetics%20Additional%20Parameter%20Infos%20Structure))
  (Class|BPVehicleCosmeticsFunctionLibrary|CosmeticActorColorReady Color "PaintedColor" AdditionalInfo (Variables|Default|GetNiagara)))

(fn ConstructionScript)
