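# Niagara system for the Boba vehicle boost cosmetic, emitting bubble-tea exhaust from boost sockets on the car's skeletal mesh.
# Particles are distributed per port across black boba sprites, brown tea ribbons, and a merged emitter for meshes and sprites.
# The merged emitter uses spawn groups to vary color, lifetime, scale and physics.
# User.bIsFrontendPreview switches rates, lifetimes and speeds for menu preview; painted-variant overrides are disabled.
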NiagaraSystem /VehicleCosmetics/Boosts/Boost_Boba/FX/NS_Boost_Boba
dependencies {
 usedDataInterfaces [&NiagaraDataInterfaceCurve, &NiagaraDataInterfaceVectorCurve, &NiagaraDataInterfaceSkeletalMesh]
 usedDynamicInputs [&AddVectorToPosition]
 usedModules [
  &self:SystemSpawnScript.SystemScriptSource.SystemScriptGraph.NiagaraNodeAssignment_0.SetVariables_#1
  &SystemState
 ]
}
emitters {
 NE_VC_Boost_Boba {
  emitterUpdateScript [
   &EmitterState()
   &VC_FilteredSocketCount()
   &SpawnRate(
    SpawnRate=&MakeCustomFloatFromBool(Bool=$User.bIsFrontendPreview, True Float=53, False Float=200)
    Use Spawn Probability=false
   )
  ]
  particleSpawnScript [
   &VC_PaintedVariantsOverride() {enabled=false}
   &InitializeParticle(
    Color=<rgba 0 0 0 1>
    Lifetime=&MakeCustomFloatFromBool(
     Bool=$User.bIsFrontendPreview
     True Float=&RandomRangeFloat(Minimum=0.7, Maximum=0.9)
     False Float=&RandomRangeFloat(Minimum=0.1, Maximum=0.3)
    )
    Position Mode=::NewEnumerator0
    Sprite Size Mode=::NewEnumerator1
    Uniform Sprite Size Max=12
    Uniform Sprite Size Min=10
    Write Material Random=true
   )
   &PartitionParticles(Number Of Partitions=$Emitter.PortCount)
   &SkeletalMeshLocation(
    Bone / Socket Index=$Particles.PartitionParticles.Partition
    Bone / Socket Sampling Mode=::NewEnumerator5
    MaskSpawnGroup=false
    Mesh Sampling Type=::NewEnumerator6
    Sampled Position Offset=<-25 0 0>
    Skeletal Mesh=$System.SkeletalMesh
   )
   &AddVelocity(
    Velocity=<-1 0 0>
    Velocity Speed Scale=&MakeCustomFloatFromBool(
     Bool=$User.bIsFrontendPreview
     True Float=&RandomRangeFloat(Minimum=500, Maximum=750)
     False Float=&RandomRangeFloat(Minimum=100, Maximum=150)
    )
   )
   &ShapeLocation(
    MaskSpawnGroup=false
    Offset=<-25 0 0>
    Offset Coordinate Space=::Local
    Offset Mode=::NewEnumerator0
    Override Seed=false
    Ring Radius=15
    Rotation Mode=::NewEnumerator1
    Shape Primitive=::NewEnumerator4
    UseOwnerScale=true
   )
  ]
  particleUpdateScript [
   &ParticleState()
   &Drag(Drag=0.5, Use Linear Drag=true, Use Rotational Drag=true) {module=Drag001}
   &InheritSourceMovement(
    Applied Position Scale=&VectorFromFloat(Value=&FloatFromCurve(FloatCurve=DataInterface() {propertyValues {Curve {keys [{interpMode=RCIM_Cubic, tangentMode=RCTM_Break, time=0, value=1}, {interpMode=RCIM_Cubic, tangentMode=RCTM_Break, time=1, value=0}]}}}))
    ApplyForce=false
    ApplyPosition=true
    ApplyPositionFalloff=false
   )
   &SolveForcesAndVelocity(Clamp Velocity=false, Limit Acceleration=false)
   &DynamicMaterialParameters(
    Index 0 Param 1=&FloatFromCurve(FloatCurve=DataInterface() {propertyValues {Curve {keys [
         {arriveTangent=-5.8452, interpMode=RCIM_Cubic, time=0, value=0}
         {arriveTangentWeight=0.0434, interpMode=RCIM_Cubic, leaveTangentWeight=0.217, tangentMode=RCTM_User, time=0.9, value=0}
         {arriveTangent=3.4164, interpMode=RCIM_Cubic, leaveTangent=3.4164, tangentMode=RCTM_User, time=0.95, value=0.75}
        ]}}})
    Index 0 Param 2=1000
    Index 0 Param 4=1
    Module.Param0WriteEnabled=true
    Module.Param1WriteEnabled=true
    Module.Param2WriteEnabled=true
    Module.Param3WriteEnabled=true
   )
   &ScaleSpriteSize(Uniform Curve Sprite Scale=DataInterface() {propertyValues {Curve {
       defaultValue=0.7
       keys [
        {arriveTangent=4.7592, interpMode=RCIM_Cubic, leaveTangent=4.7592, leaveTangentWeight=0.0333, tangentMode=RCTM_User, time=0, value=0.2}
        {arriveTangentWeight=0.0471, interpMode=RCIM_Cubic, leaveTangentWeight=0.3771, tangentMode=RCTM_User, time=0.2, value=1}
        {arriveTangentWeight=0.2663, interpMode=RCIM_Cubic, leaveTangentWeight=0.0205, tangentMode=RCTM_User, time=0.85, value=1}
        {arriveTangentWeight=0.3986, interpMode=RCIM_Cubic, leaveTangentWeight=0.0498, tangentMode=RCTM_User, time=0.9, value=0.3}
        {arriveTangentWeight=0.1537, interpMode=RCIM_Cubic, leaveTangentWeight=0.1537, tangentMode=RCTM_User, time=0.95, value=1.5}
        {arriveTangent=-4.6964, arriveTangentWeight=0.0333, interpMode=RCIM_Cubic, leaveTangent=-4.6964, tangentMode=RCTM_User, time=1, value=0.1}
       ]
      }}})
  ]
  rendererClasses [&NiagaraSpriteRendererProperties]
  renderers [
   {rendererClass=&NiagaraSpriteRendererProperties}
   {rendererClass=&NiagaraSpriteRendererProperties, rendererIndex=1}
  ]
 }
 NE_VC_Boost_Sprite_Parent007 {
  emitterUpdateScript [
   &EmitterState(Inactive Response=::NewEnumerator1, Life Cycle Mode=::NewEnumerator1, UseLoopDelay=false)
   &SpawnRate(
    SpawnRate=&MakeCustomFloatFromBool(Bool=$User.bIsFrontendPreview, True Float=28, False Float=44)
    Use Spawn Probability=false
   ) {module=SpawnRate003}
   &SpawnRate(
    SpawnGroup=1
    SpawnRate=&MakeCustomFloatFromBool(Bool=$User.bIsFrontendPreview, True Float=28, False Float=44)
    Use Spawn Probability=false
   ) {module=SpawnRate001}
   &SpawnRate(SpawnGroup=2, SpawnRate=150, Use Spawn Probability=false) {module=SpawnRate002}
   &VC_FilteredSocketCount()
   &SpawnRate(
    SpawnRate=&MakeCustomFloatFromBool(Bool=$User.bIsFrontendPreview, True Float=28, False Float=44)
    Use Spawn Probability=false
   ) {enabled=false}
  ]
  particleSpawnScript [
   /SetVariables(Particles.VisibilityTag=&SpawnGroup()) {isSetParameters=true}
   &VC_PaintedVariantsOverride() {enabled=false}
   &InitializeParticle(
    Color=/VehicleCosmetics/FX/Niagara/Modules/MergeEmitters/VC_Dynamic_ColorBySpawnGroup(
     Color SG 0:LinearColor=<rgba 0.4479 0.0895 0.009 0.45>
     MaximumSpawnGroups:NiagaraInt32=3
     Color SG 1:LinearColor=<rgba 0.4479 0.0895 0.009 0.1>
     Color SG 2:LinearColor=<rgba 0 0 0 1>
     Color SG 3:LinearColor=<rgba 1 1 1 1>
     Color SG 4:LinearColor=<rgba 1 1 1 1>
    )
    Lifetime=/VehicleCosmetics/FX/Niagara/Modules/MergeEmitters/VC_FloatBySpawnGroup(
     SpawnGroup or ExecIndex:NiagaraInt32=0
     CustomInt:NiagaraInt32=0
     MaximumSpawnGroups:NiagaraInt32=3
     Float SG 0:NiagaraFloat=&MakeCustomFloatFromBool(
      Bool=$User.bIsFrontendPreview
      True Float=&RandomRangeFloat(Minimum=0.5, Maximum=0.6)
      False Float=&RandomRangeFloat(Minimum=0.4, Maximum=0.5)
     )
     Float SG 1:NiagaraFloat=&RandomRangeFloat(Minimum=0.4, Maximum=0.5)
     Float SG 2:NiagaraFloat=&RandomRangeFloat(Minimum=0.3, Maximum=0.5)
     Float SG 3:NiagaraFloat=0
     Float SG 4:NiagaraFloat=0
     Float SG 5:NiagaraFloat=0
    )
    Mesh Scale=/VehicleCosmetics/FX/Niagara/Modules/MergeEmitters/VC_VectorBySpawnGroup(
     SpawnGroup or ExecIndex:NiagaraInt32=0
     CustomInt:NiagaraInt32=0
     MaximumSpawnGroups:NiagaraInt32=2
     Vector Group 0:Vector3f=<0.15 0.15 0.18>
     Vector Group 1:Vector3f=<1.5 1.5 2.1>
     Vector Group 2:Vector3f=<0 0 0>
     Vector Group 3:Vector3f=<0 0 0>
     Vector Group 4:Vector3f=<0 0 0>
     Vector Group 5:Vector3f=<0 0 0>
    )
    Mesh Scale Mode=::NewEnumerator4
    Sprite Rotation Mode=::NewEnumerator2
    Sprite Size Mode=::NewEnumerator3
    Uniform Sprite Size=0.1
    UsePositionOffset=false
    Write Material Random=true
   )
   &PartitionParticles(Number Of Partitions=$Emitter.PortCount, Partition Duration=0.05, Partition Mode:ENiagara_PartitionMode=::NewEnumerator2)
   &SkeletalMeshLocation(
    Bone / Socket Index=$Particles.PartitionParticles.Partition
    Bone / Socket Sampling Mode=::NewEnumerator5
    MaskSpawnGroup=false
    Mesh Sampling Type=::NewEnumerator6
    Orientation Sampling=::NewEnumerator1
    Sampled Position Offset=&MaskVectorBySpawnGroup(Masked Vector=<-30 0 0>, Passthrough Vector=<-32 0 0>, Spawn Group Mask=1)
    Skeletal Mesh=$System.SkeletalMesh
    Use Sampled Mesh Attribute as Ribbon ID=false
   )
   &ShapeLocation(
    MaskSpawnGroup=true
    Offset=<5 0 0>
    Offset Mode=::NewEnumerator0
    Override Seed=false
    Spawn Group Mask=2
    Sphere Radius=1
    UseOwnerScale=true
   )
   &AddVelocity(
    Cone Angle=60
    Cone Axis=<-1 0 0.2>
    Inner Cone Angle=20
    Module.UseConeFalloff=false
    Module.UseVelDistribution=false
    Override Seed=false
    Velocity Mode=::NewEnumerator2
    Velocity Speed=&VC_FloatBySpawnGroup(
     SpawnGroup or ExecIndex:NiagaraInt32=0
     CustomInt:NiagaraInt32=0
     MaximumSpawnGroups:NiagaraInt32=3
     Float SG 0:NiagaraFloat=0
     Float SG 1:NiagaraFloat=0
     Float SG 2:NiagaraFloat=&RandomRangeFloat(Minimum=200, Maximum=800)
     Float SG 3:NiagaraFloat=0
     Float SG 4:NiagaraFloat=0
     Float SG 5:NiagaraFloat=0
    )
   )
   &InitialMeshOrientation(
    Mesh Orientation Mode=::NewEnumerator7
    Rotation=&RandomRangeVector(Minimum=<-1 0 0>, Maximum=<1 0 0>)
    Use Rotation Vector=true
   )
  ]
  particleUpdateScript [
   &ParticleState()
   &ScaleMeshSize(Scale Factor=&VC_VectorBySpawnGroup(
     SpawnGroup or ExecIndex:NiagaraInt32=0
     CustomInt:NiagaraInt32=0
     MaximumSpawnGroups:NiagaraInt32=2
     Vector Group 0:Vector3f=&VectorFromCurve(VectorCurve=DataInterface() {propertyValues {
        XCurve {keys [<at 0 0.3>,<at 0.45 0.7>,<at 0.9 1>,<at 0.91 0 RCIM_Constant>]}
        YCurve {keys [<at 0 0.3>,<at 0.45 0.7>,<at 0.9 1>,<at 0.91 0 RCIM_Constant>]}
        ZCurve {keys [<at 0 0.1>,<at 0.45 0.375>,<at 0.9 1>,<at 0.91 0 RCIM_Constant>]}
       }})
     Vector Group 1:Vector3f=&VectorFromCurve(VectorCurve=DataInterface() {propertyValues {
        XCurve {keys [
          {arriveTangent=0.2699, interpMode=RCIM_Cubic, leaveTangent=0.2699, leaveTangentWeight=0.3333, tangentMode=RCTM_User, time=0, value=0.5}
          {arriveTangentWeight=0.3333, interpMode=RCIM_Cubic, tangentMode=RCTM_User, time=1, value=1}
         ]}
        YCurve {keys [
          {arriveTangent=0.2699, interpMode=RCIM_Cubic, leaveTangent=0.2699, leaveTangentWeight=0.3333, tangentMode=RCTM_User, time=0, value=0.5}
          {arriveTangentWeight=0.3333, interpMode=RCIM_Cubic, tangentMode=RCTM_User, time=1, value=1}
         ]}
        ZCurve {keys [
          {arriveTangent=0.9827, interpMode=RCIM_Cubic, leaveTangent=0.9827, tangentMode=RCTM_User, time=0, value=0.2}
          {arriveTangent=0.3749, interpMode=RCIM_Cubic, leaveTangent=0.3749, tangentMode=RCTM_User, time=1, value=1.5}
         ]}
       }})
     Vector Group 2:Vector3f=<0 0 0>
     Vector Group 3:Vector3f=<0 0 0>
     Vector Group 4:Vector3f=<0 0 0>
     Vector Group 5:Vector3f=<0 0 0>
    )) {enabled=false, module=ScaleMeshSize001}
   &ScaleMeshSize(Scale Factor=&VC_VectorBySpawnGroup(
     SpawnGroup or ExecIndex:NiagaraInt32=0
     CustomInt:NiagaraInt32=0
     MaximumSpawnGroups:NiagaraInt32=2
     Vector Group 0:Vector3f=&VectorFromCurve(VectorCurve=DataInterface() {propertyValues {
        XCurve {keys [<at 0 1>,<at 0.4 1.9>,<at 0.9 3>,<at 0.91 0 RCIM_Constant>]}
        YCurve {keys [<at 0 1>,<at 0.4 1.9>,<at 0.9 3>,<at 0.91 0 RCIM_Constant>]}
        ZCurve {keys [<at 0 1>,<at 0.4 3.5>,<at 0.9 8>,<at 0.91 0 RCIM_Constant>]}
       }})
     Vector Group 1:Vector3f=&VectorFromCurve(VectorCurve=DataInterface() {propertyValues {
        XCurve {keys [
          {arriveTangent=0.2699, interpMode=RCIM_Cubic, leaveTangent=0.2699, leaveTangentWeight=0.3333, tangentMode=RCTM_User, time=0, value=1}
          {arriveTangentWeight=0.3333, interpMode=RCIM_Cubic, tangentMode=RCTM_User, time=1, value=1.6}
         ]}
        YCurve {keys [
          {arriveTangent=0.2699, interpMode=RCIM_Cubic, leaveTangent=0.2699, leaveTangentWeight=0.3333, tangentMode=RCTM_User, time=0, value=1}
          {arriveTangentWeight=0.3333, interpMode=RCIM_Cubic, tangentMode=RCTM_User, time=1, value=1.6}
         ]}
        ZCurve {keys [
          {arriveTangent=0.9827, interpMode=RCIM_Cubic, leaveTangent=0.9827, tangentMode=RCTM_User, time=0, value=1}
          {arriveTangent=0.3749, interpMode=RCIM_Cubic, leaveTangent=0.3749, tangentMode=RCTM_User, time=1, value=4}
         ]}
       }})
     Vector Group 2:Vector3f=<0 0 0>
     Vector Group 3:Vector3f=<0 0 0>
     Vector Group 4:Vector3f=<0 0 0>
     Vector Group 5:Vector3f=<0 0 0>
    ))
   &GravityForce(Gravity=&Multiply_VectorByFloat(
     Vector=<0 0 -980>
     Float=&VC_FloatBySpawnGroup(
      SpawnGroup or ExecIndex:NiagaraInt32=0
      CustomInt:NiagaraInt32=0
      MaximumSpawnGroups:NiagaraInt32=3
      Float SG 0:NiagaraFloat=0
      Float SG 1:NiagaraFloat=0
      Float SG 2:NiagaraFloat=0.5
      Float SG 3:NiagaraFloat=0
      Float SG 4:NiagaraFloat=0
      Float SG 5:NiagaraFloat=0
     )
    ))
   &Drag(
    Drag=&VC_FloatBySpawnGroup(
     SpawnGroup or ExecIndex:NiagaraInt32=0
     CustomInt:NiagaraInt32=0
     MaximumSpawnGroups:NiagaraInt32=3
     Float SG 0:NiagaraFloat=0
     Float SG 1:NiagaraFloat=0
     Float SG 2:NiagaraFloat=3
     Float SG 3:NiagaraFloat=0
     Float SG 4:NiagaraFloat=0
     Float SG 5:NiagaraFloat=0
    )
    Use Linear Drag=true
    Use Rotational Drag=true
   )
   &SolveForcesAndVelocity(Clamp Velocity=false, Limit Acceleration=false)
   &DynamicMaterialParameters(
    Index 2 Param 1=&FloatFromCurve(
     FloatCurve=DataInterface() {propertyValues {Curve {keys [
         {arriveTangent=-5.8452, interpMode=RCIM_Cubic, time=0, value=1}
         {arriveTangentWeight=0.0434, interpMode=RCIM_Cubic, leaveTangentWeight=0.217, tangentMode=RCTM_User, time=0.2, value=0}
         {arriveTangent=3.4164, interpMode=RCIM_Cubic, leaveTangent=3.4164, tangentMode=RCTM_User, time=0.8, value=0.7}
         <at 1 0 RCIM_Cubic>
        ]}}}
     Scale Curve=1.5
    )
    Index 2 Param 2=50
    Index 2 Param 3=500
    Index 2 Param 4=5
    Module.Index0Float0Write=true
    Module.Index0Float1Write=true
    Module.Index0Float2Write=true
    Module.Index0Float3Write=false
    Module.Index2Float0Write=true
    Module.Index2Float1Write=true
    Module.Index2Float2Write=true
    Module.Index2Float3Write=true
    Module.Index3Float0Write=true
    Module.Index3Float1Write=true
    Module.Index3Float2Write=true
    Module.Index3Float3Write=true
    Write Parameter Index 2=::NewEnumerator1
    Write Parameter Index 3=::NewEnumerator1
   )
   &ScaleSpriteSize(
    Uniform Curve Scale=&RandomRangeFloat(Minimum=1, Maximum=3)
    Uniform Curve Sprite Scale=DataInterface() {propertyValues {Curve {keys [
        {arriveTangent=6.7005, interpMode=RCIM_Cubic, leaveTangent=6.7005, tangentMode=RCTM_User, time=0, value=1}
        <at 0.16 30>
        {interpMode=RCIM_Cubic, tangentMode=RCTM_User, time=0.75, value=30}
        {arriveTangent=-4.4982, interpMode=RCIM_Cubic, leaveTangent=-4.4982, tangentMode=RCTM_User, time=1, value=1}
       ]}}}
   )
   &ScaleColor(
    Scale Alpha=&VC_FloatBySpawnGroup(
     SpawnGroup or ExecIndex:NiagaraInt32=0
     CustomInt:NiagaraInt32=0
     MaximumSpawnGroups:NiagaraInt32=3
     Float SG 0:NiagaraFloat=&FloatFromCurve(FloatCurve=DataInterface() {propertyValues {Curve {
         defaultValue=1
         keys [
          {arriveTangentWeight=0.0473, interpMode=RCIM_Cubic, leaveTangentWeight=0.426, tangentMode=RCTM_User, time=0, value=1}
          {arriveTangentWeight=0.3018, interpMode=RCIM_Cubic, tangentMode=RCTM_User, time=1, value=0}
         ]
        }}})
     Float SG 1:NiagaraFloat=&FloatFromCurve(FloatCurve=DataInterface() {propertyValues {Curve {
         defaultValue=1
         keys [
          {arriveTangent=43.5199, interpMode=RCIM_Cubic, leaveTangent=43.5199, tangentMode=RCTM_User, time=0, value=1}
          {interpMode=RCIM_Cubic, tangentMode=RCTM_User, time=0.2, value=10}
          {arriveTangent=-11.25, interpMode=RCIM_Cubic, leaveTangent=-11.25, time=0.5023, value=7}
          {arriveTangentWeight=0.2667, interpMode=RCIM_Cubic, leaveTangent=-1.7044, time=1, value=1}
         ]
        }}})
     Float SG 2:NiagaraFloat=1
     Float SG 3:NiagaraFloat=0
     Float SG 4:NiagaraFloat=0
     Float SG 5:NiagaraFloat=0
    )
    ScaleA=true
    ScaleRGB=true
   )
  ]
  rendererClasses [&NiagaraMeshRendererProperties, &NiagaraSpriteRendererProperties]
  renderers [
   {rendererClass=&NiagaraMeshRendererProperties, rendererIndex=0}
   {rendererClass=&NiagaraMeshRendererProperties, rendererIndex=1}
   {rendererClass=&NiagaraSpriteRendererProperties, rendererIndex=2}
   {rendererClass=&NiagaraSpriteRendererProperties, rendererIndex=3}
  ]
 }
 NE_VC_Boost_Tea {
  emitterSpawnScript [/SetVariables(Emitter.EatawyPan:NiagaraFloat=&MakeCustomFloatFromBool(Bool=$User.bIsFrontendPreview, True Float=-2, False Float=-4)) {isSetParameters=true}]
  emitterUpdateScript [
   &EmitterState()
   &VC_FilteredSocketCount()
   &VC_RateQualityHandler() {enabled=false}
   &VC_SpawnRibbon_v1(Number of Ribbons=$Emitter.PortCount, SpawnRate=60)
  ]
  particleSpawnScript [
   &VC_PaintedVariantsOverride() {enabled=false}
   &VC_InitializeRibbon_v1(
    Color=$Particles.Color
    Enable Skeletal Position=true
    Lifetime=&MakeCustomFloatFromBool(Bool=$User.bIsFrontendPreview, True Float=0.8, False Float=0.6)
    Offset Amount=-15
    Offset Axis:ENiagaraOrientationAxis=::XAxis
    Ribbon Width=15
   )
   &AddVelocity(Velocity=<-2500 0 0>) {module=AddVelocity001}
   &Color(Color=<rgba 0.4479 0.0895 0.009 1>)
  ]
  particleUpdateScript [
   &ParticleState()
   &InheritSourceMovement(
    Applied Position Scale=&VectorFromFloat(Value=&FloatFromCurve(FloatCurve=DataInterface() {propertyValues {Curve {keys [{interpMode=RCIM_Cubic, tangentMode=RCTM_Break, time=0, value=1}, {interpMode=RCIM_Cubic, tangentMode=RCTM_Break, time=1, value=0}]}}}))
    ApplyForce=false
    ApplyPosition=true
    ApplyPositionFalloff=false
   )
   &SolveForcesAndVelocity(Clamp Velocity=false, Limit Acceleration=false)
   &Color(Color=&MultiplyLinearColor(
     Linear Color=$Particles.Color
     Scale Factor=&MakeVector4_FromLinearColor(LinearColor=&MakeLinearColorFromVectorAndFloat(
       "Vector (RGB)"=<1 1 1>
       "Float (Alpha)"=&Multiply_Float(
        A=&FloatFromCurve(FloatCurve=DataInterface() {propertyValues {Curve {
            defaultValue=0
            keys [
             {arriveTangent=-0.8146, interpMode=RCIM_Cubic, leaveTangent=-0.8146, leaveTangentWeight=0.3333, tangentMode=RCTM_User, time=0, value=1}
             {arriveTangent=-1.737, arriveTangentWeight=0.3333, interpMode=RCIM_Cubic, leaveTangent=-1.737, tangentMode=RCTM_User, time=0.85, value=0}
            ]
           }}})
        B=&FloatFromCurve(FloatCurve=DataInterface() {propertyValues {Curve {keys [<at 0 0>,<at 0.5 1>]}}}, CurveIndex=$Emitter.Age)
       )
      ))
    )) {module=Color001}
   &ScaleRibbonWidth(Ribbon Width Scale=&Lerp_Float(
     A=1
     B=4.5
     Alpha=&FloatFromCurve(FloatCurve=DataInterface() {propertyValues {Curve {keys [
          {arriveTangent=1.0179, interpMode=RCIM_Cubic, leaveTangent=1.0179, leaveTangentWeight=0.3333, tangentMode=RCTM_User, time=0, value=0}
          {arriveTangent=1.0828, interpMode=RCIM_Cubic, leaveTangent=1.0828, tangentMode=RCTM_User, time=0.4546, value=0.6761}
          {arriveTangent=0.2669, arriveTangentWeight=0.3333, interpMode=RCIM_Cubic, leaveTangent=0.2669, tangentMode=RCTM_User, time=1, value=1}
         ]}}})
    ))
   &VC_UpdateRibbon_v1(Enable Skeletal Position=true)
   &ScaleColor(
    Scale Alpha=&FloatFromCurve(FloatCurve=DataInterface() {propertyValues {Curve {
        defaultValue=0
        keys [
         {arriveTangent=-0.8146, interpMode=RCIM_Cubic, leaveTangent=-0.8146, leaveTangentWeight=0.3333, tangentMode=RCTM_User, time=0, value=1}
         {arriveTangent=-1.737, arriveTangentWeight=0.3333, interpMode=RCIM_Cubic, leaveTangent=-1.737, tangentMode=RCTM_User, time=0.85, value=0}
        ]
       }}})
    ScaleA=true
    ScaleRGB=true
   ) {enabled=false}
   &DynamicMaterialParameters(
    Index 0 Param 2=&FloatFromCurve(FloatCurve=DataInterface() {propertyValues {Curve {keys [
         <at 0.05 1 RCIM_Cubic>
         {arriveTangentWeight=0.255, interpMode=RCIM_Cubic, leaveTangentWeight=0.085, tangentMode=RCTM_User, time=0.2, value=0}
         {arriveTangentWeight=0.1124, interpMode=RCIM_Cubic, leaveTangentWeight=0.2247, tangentMode=RCTM_User, time=0.4, value=0}
         {arriveTangentWeight=0.3333, interpMode=RCIM_Cubic, tangentMode=RCTM_User, time=0.45, value=1}
        ]}}})
    Index 0 Param 3=&MakeCustomFloatFromBool(
     Bool=$User.bIsFrontendPreview
     True Float=&FloatFromCurve(FloatCurve=DataInterface() {propertyValues {Curve {keys [<at 0.2 1>,<at 0.45 0>]}}})
     False Float=2
    )
    Index 0 Param 4=&FloatFromCurve(FloatCurve=DataInterface() {propertyValues {Curve {keys [
         {arriveTangentWeight=0.255, interpMode=RCIM_Cubic, leaveTangentWeight=0.085, tangentMode=RCTM_User, time=0.25, value=0}
         {
          arriveTangent=3.9179
          arriveTangentWeight=0.1124
          interpMode=RCIM_Cubic
          leaveTangent=3.9179
          leaveTangentWeight=0.2247
          tangentMode=RCTM_User
          time=0.45
          value=0.5
         }
         {arriveTangentWeight=0.3333, interpMode=RCIM_Cubic, tangentMode=RCTM_User, time=0.7, value=1}
        ]}}})
    Module.Index0Float0Write=true
    Module.Index0Float1Write=true
    Module.Index0Float2Write=true
    Module.Index0Float3Write=true
   )
  ]
  rendererClasses [&NiagaraRibbonRendererProperties]
  renderers [
   {rendererClass=&NiagaraRibbonRendererProperties}
   {rendererClass=&NiagaraRibbonRendererProperties, rendererIndex=1}
  ]
 }
}
summary {
 systemName=NS_Boost_Boba
 emitters [
  {emitterName=NE_VC_Boost_Boba, rendererClasses [&NiagaraSpriteRendererProperties]}
  {
   emitterName=NE_VC_Boost_Sprite_Parent007
   rendererClasses [&NiagaraMeshRendererProperties, &NiagaraSpriteRendererProperties]
  }
  {emitterName=NE_VC_Boost_Tea, rendererClasses [&NiagaraRibbonRendererProperties]}
 ]
 userVariables [
  User.bIsFrontendPreview {type=NiagaraBool}
  User.bIsPainted {type=NiagaraBool}
  User.ColorVariantID {type=NiagaraInt32}
  User.DTCarOffset {type=Vector3f, description="", defaultValue {struct=Vector3f, value=<0 0 0>}}
  User.PortCount {type=NiagaraInt32, defaultValue=2}
 ]
}
